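# Niagara system for the Casino Reaper glider's deploy effect, drawing short-lived, translucent ribbon trails.
# A single CornerTrails emitter partitions particles into 20 anim trails, using modulo and division by five to pair five socket indices with four corner offsets.
# Socket-reader modules anchor trails, which drift outward and upward while partly inheriting source movement.
# Ribbons widen and fade over life, with color and alpha adjusted by time of day.
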NiagaraSystem /BRCosmetics/Gadgets/Assets/VinderTech_GliderChute/Glider_CasinoReaper/FX/NS_Glider_CasinoReaper_Deploy
dependencies {
 usedDataInterfaces [&NiagaraDataInterfaceCurve, &NiagaraDataInterfaceSocketReader]
 usedDynamicInputs []
 usedModules [
  &self:SystemSpawnScript.SystemScriptSource.SystemScriptGraph.NiagaraNodeAssignment_0.SetVariables_#1
  &SystemState
  &self:SystemSpawnScript.SystemScriptSource.SystemScriptGraph.NiagaraNodeAssignment_1.SetVariables_#2
 ]
}
emitters {CornerTrails {
  emitterUpdateScript [
   &EmitterState()
   &FN_SpawnAnimTrail_v2(Number of Anim Trails=20)
  ]
  particleSpawnScript [
   &PartitionParticles(Number Of Partitions=20)
   /Game/Effects/Niagara/Modules/AnimTrail/FN_InitializeAnimTrail/FN_InitializeAnimTrailBoneIndecies(
    AnimTrail Selection:NiagaraInt32=$Particles.PartitionParticles.Partition
    First Bone or Socket Index:NiagaraInt32=&ModuloInt(A=$Particles.PartitionParticles.Partition, B=5)
    Second Bone or Socket Index:NiagaraInt32=&ModuloInt(A=$Particles.PartitionParticles.Partition, B=5)
   )
   /Game/Effects/Niagara/Modules/AnimTrail/FN_InitializeAnimTrail/FN_InitializeAnimTrail_SocketReader/FN_InitializeAnimTrail_SocketReader(
    AnimTrailWidth=50
    Color:LinearColor=&MultiplyLinearColor(
     Linear Color=<rgba 0.6239 0.8452 0.974 0.35>
     Scale Factor=&MakeFloat4FromVectorAndFloat(
      "Vector (RGB)"=<1 1 1>
      "Float (Alpha)"=&FloatFromCurve(FloatCurve=DataInterface() {propertyValues {Curve {keys [<at 0.2 1>,<at 0.25 1>]}}}, CurveIndex=$System.Age)
     )
    )
    Lifetime=0.35
    Socket Reader:NiagaraDataInterfaceSocketReader=$System.Socket Reader
   )
   &SocketLocation(
    SamplePosOffset=true
    Sampled Position Offset=&MergeHelper_VectorSwitch(
     0=<64 40 0>
     1=<64 -40 0>
     2=<-64 40 0>
     3=<-64 -40 0>
     Index=/Niagara/DynamicInputs/Divide/Divide_Int(A:NiagaraInt32=$Particles.PartitionParticles.Partition, B:NiagaraInt32=5)
     Num=::NewEnumerator2
    )
    Socket Index=&ModuloInt(A=$Particles.PartitionParticles.Partition, B=5)
    Socket Reader=$System.Socket Reader
    Socket Sampling Mode=::NewEnumerator3
    VelocitySpeedLimit=false
   )
   &AddVelocity(
    Velocity=&MergeHelper_VectorSwitch(
     0=<100 100 100>
     1=<100 -100 100>
     2=<-100 100 100>
     3=<-100 -100 100>
     Index=&Divide_Int(A:NiagaraInt32=$Particles.PartitionParticles.Partition, B:NiagaraInt32=5)
     Num=::NewEnumerator2
    )
    Velocity Speed Scale=0.5
   )
  ]
  particleUpdateScript [
   &ScaleRibbonWidth(Ribbon Width Scale=&FloatFromCurve(FloatCurve=DataInterface() {propertyValues {Curve {keys [
         {arriveTangent=1.4312, interpMode=RCIM_Cubic, leaveTangent=1.4312, tangentMode=RCTM_User, time=0, value=0.3}
         {arriveTangent=0.7, interpMode=RCIM_Cubic, leaveTangent=0.7, time=0.3, value=1}
         {interpMode=RCIM_Cubic, leaveTangent=-0.9664, time=1, value=1}
        ]}}}))
   &ScaleColor(
    Scale Alpha=&Multiply_Float(
     A=&FloatFromCurve(
      FloatCurve=DataInterface() {propertyValues {Curve {keys [{interpMode=RCIM_Cubic, leaveTangent=-2.3646, tangentMode=RCTM_Break, time=0, value=1}, <at 1 0.05>]}}}
      Scale Curve=$System.AlphaFade
     )
     B=&FNTimeOfDayScalar(Night Scale=0.5, Morning Scale=0.6, Evening Scale=0.5)
    )
    Scale RGB=&MakeVectorFromLinearColorRGB(Color=&FNTimeOfDayColor(
      Night Color=<rgba 0.4463 0.62 1 1>
      Morning Color=<rgba 1 0.6899 0.5573 1>
      Day Color=<rgba 1 0.7878 0.6667 1>
      Evening Color=<rgba 1 0.6291 0.4464 1>
     ))
    ScaleA=true
    ScaleRGB=true
   ) {module=ScaleColor002}
   /Game/Effects/Niagara/Modules/AnimTrail/FN_UpdateAnimTrail/FN_UpdateAnimTrail_SocketReader/FN_UpdateAnimTrail_SocketReader(Socket Reader:NiagaraDataInterfaceSocketReader=$System.Socket Reader)
   &InheritSourceMovement(Applied Position Scale=<0.8 0.8 0.8>, ApplyForce=false, ApplyPosition=true, ApplyPositionFalloff=false)
   &SolveForcesAndVelocity(Clamp Velocity=false, Limit Acceleration=false)
  ]
  rendererClasses [&NiagaraRibbonRendererProperties]
  renderers [{rendererClass=&NiagaraRibbonRendererProperties}]
 }}
summary {
 systemName=NS_Glider_CasinoReaper_Deploy
 emitters [{emitterName=CornerTrails, rendererClasses [&NiagaraRibbonRendererProperties]}]
 userVariables [User.RibbonIDOffset {type=NiagaraInt32}]
}
