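# Niagara system for the Casino Reaper glider's flare, a short-lived sprite burst at a socket on the glider.
# User.ColorIndex modulo 5 picks the socket, and the initial color comes from a system-level TOD value.
# The sprite swells, shrinks and rotates with a camera offset, while nested switches on VariantIndex and ColorIndex parity pick its color-over-life curve.
# An earlier copy of that color-scaling module remains disabled.
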NiagaraSystem /BRCosmetics/Gadgets/Assets/VinderTech_GliderChute/Glider_CasinoReaper/FX/NS_Glider_CasinoReaper_Flare
dependencies {
 usedDataInterfaces [&NiagaraDataInterfaceCurve, &NiagaraDataInterfaceColorCurve, &NiagaraDataInterfaceSocketReader]
 usedDynamicInputs []
 usedModules [
  &self:SystemSpawnScript.SystemScriptSource.SystemScriptGraph.NiagaraNodeAssignment_0.SetVariables_#1
  &self:SystemSpawnScript.SystemScriptSource.SystemScriptGraph.NiagaraNodeAssignment_1.SetVariables_#2
  &SystemState
 ]
}
emitters {Flare {
  emitterUpdateScript [
   &EmitterState()
   &SpawnBurst_Instantaneous(Use Spawn Probability=false, UseLoopCountLimit=false)
  ]
  particleSpawnScript [
   &InitializeParticle(
    Color=$System.TOD
    Lifetime=&RandomRangeFloat(Minimum=0.3, Maximum=0.4)
    Sprite Rotation Angle=45
    Sprite Rotation Mode=::NewEnumerator0
    Sprite Size=&Multiply_Vector2D_ByFloat(Float=&RandomRangeFloat(Minimum=100, Maximum=100), Vector2D=<1.1 1>)
    Sprite Size Mode=::NewEnumerator4
    UsePositionOffset=true
    Write Material Random=true
   )
   &SocketLocation(
    Orientation Sampling=::NewEnumerator1
    SamplePosOffset=false
    Socket Index=&ModuloInt(A=$User.ColorIndex, B=5)
    Socket Reader=$System.Socket Reader
    Socket Sampling Mode=::NewEnumerator3
    VelocitySpeedLimit=false
   )
  ]
  particleUpdateScript [
   &ParticleState()
   &ScaleSpriteSize(Uniform Curve Sprite Scale=DataInterface() {propertyValues {Curve {keys [
        {arriveTangent=0.9776, interpMode=RCIM_Cubic, leaveTangent=0.9776, tangentMode=RCTM_User, time=0, value=0}
        <at 0.3 1 RCIM_Cubic>
        {arriveTangent=-0.8026, interpMode=RCIM_Cubic, leaveTangent=-0.8026, tangentMode=RCTM_User, time=1, value=0}
       ]}}})
   &SpriteRotationRate(Rotation Rate=&FloatFromCurve(FloatCurve=DataInterface() {propertyValues {Curve {keys [<at 0.2 1 RCIM_Cubic>,<at 0.4 180 RCIM_Cubic>]}}}))
   &CameraOffset(Camera Offset Amount=100) {module=CameraOffset001}
   &ScaleColor(
    Scale Alpha=&FloatFromCurve(FloatCurve=DataInterface() {propertyValues {Curve {keys [<at 0.6 1>,<at 1 2>]}}})
    Scale RGB=&MakeVectorFromLinearColorRGB(Color=&MergeHelper_ColorSwitch(
      0=&MergeHelper_ColorSwitch(
       0=&ColorFromCurve(ColorCurve=DataInterface() {propertyValues {
          AlphaCurve {keys [<at 0 1 RCIM_Cubic>,<at 1 1 RCIM_Cubic>]}
          BlueCurve {keys [<at 0.0283 5>,<at 0.4401 0 RCIM_Cubic>]}
          GreenCurve {keys [<at 0.0283 4.7557>,<at 0.4401 0 RCIM_Cubic>]}
          RedCurve {keys [<at 0.0283 3.1251>,<at 0.4401 0 RCIM_Cubic>]}
         }})
       1=&ColorFromCurve(ColorCurve=DataInterface() {propertyValues {
          AlphaCurve {keys [<at 0 1 RCIM_Cubic>,<at 1 1 RCIM_Cubic>]}
          BlueCurve {keys [<at 0.0394 0.9982>,<at 0.2509 0.2093>,<at 0.6029 0.0837>,<at 0.9852 0.0219>]}
          GreenCurve {keys [<at 0.0394 1.2586>,<at 0.2509 0>,<at 0.6029 0>,<at 0.9852 0>]}
          RedCurve {keys [<at 0.0394 5>,<at 0.2509 5>,<at 0.6029 0.5417>,<at 0.9852 0.151>]}
         }})
       2=&ColorFromCurve(ColorCurve=DataInterface() {propertyValues {
          AlphaCurve {keys [<at 0 1 RCIM_Cubic>,<at 1 1 RCIM_Cubic>]}
          BlueCurve {keys [<at 0.0283 5>,<at 0.4401 0 RCIM_Cubic>]}
          GreenCurve {keys [<at 0.0283 4.7557>,<at 0.4401 0 RCIM_Cubic>]}
          RedCurve {keys [<at 0.0283 3.1251>,<at 0.4401 0 RCIM_Cubic>]}
         }})
       3=?
       Index=&ModuloInt(A=$User.ColorIndex, B=2)
       Num=::NewEnumerator0
      )
      1=&MergeHelper_ColorSwitch(
       0=&ColorFromCurve(ColorCurve=DataInterface() {propertyValues {
          AlphaCurve {keys [<at 0 1 RCIM_Cubic>,<at 1 1 RCIM_Cubic>]}
          BlueCurve {keys [<at 0.0394 0.9982>,<at 0.2509 5>,<at 0.6029 0.5417>,<at 0.9852 0.151>]}
          GreenCurve {keys [<at 0.0394 1.2586>,<at 0.2509 4.6848>,<at 0.6029 0.2545>,<at 0.9852 0.025>]}
          RedCurve {keys [<at 0.0394 5>,<at 0.2509 0>,<at 0.6029 0.0591>,<at 0.9852 0.0552>]}
         }})
       1=&ColorFromCurve(ColorCurve=DataInterface() {propertyValues {
          AlphaCurve {keys [<at 0 1 RCIM_Cubic>,<at 1 1 RCIM_Cubic>]}
          BlueCurve {keys [<at 0.0394 0.9982>,<at 0.2509 5>,<at 0.6029 0.5417>,<at 0.9852 0.151>]}
          GreenCurve {keys [<at 0.0394 1.2586>,<at 0.2509 4.6848>,<at 0.6029 0.2545>,<at 0.9852 0.025>]}
          RedCurve {keys [<at 0.0394 5>,<at 0.2509 0>,<at 0.6029 0.0591>,<at 0.9852 0.0552>]}
         }})
       2=&ColorFromCurve(ColorCurve=DataInterface() {propertyValues {
          AlphaCurve {keys [<at 0 1 RCIM_Cubic>,<at 1 1 RCIM_Cubic>]}
          BlueCurve {keys [<at 0.0283 5>,<at 0.4401 0 RCIM_Cubic>]}
          GreenCurve {keys [<at 0.0283 4.7557>,<at 0.4401 0 RCIM_Cubic>]}
          RedCurve {keys [<at 0.0283 3.1251>,<at 0.4401 0 RCIM_Cubic>]}
         }})
       3=?
       Index=&ModuloInt(A=$User.ColorIndex, B=2)
       Num=::NewEnumerator0
      )
      2=?
      3=?
      Index=&MakeIntFromFloat(Float=$User.VariantIndex)
      Num=::NewEnumerator0
     ))
    ScaleA=true
    ScaleRGB=true
   ) {enabled=false}
   &ScaleColor(
    Scale Alpha=&FloatFromCurve(FloatCurve=DataInterface() {propertyValues {Curve {keys [<at 0.6 1>,<at 1 2>]}}})
    Scale RGB=&MakeVectorFromLinearColorRGB(Color=&MergeHelper_ColorSwitch(
      0=&MergeHelper_ColorSwitch(
       0=&ColorFromCurve(ColorCurve=DataInterface() {propertyValues {
          AlphaCurve {keys [<at 0 1 RCIM_Cubic>,<at 1 1 RCIM_Cubic>]}
          BlueCurve {keys [<at 0.0283 5>,<at 0.4401 0 RCIM_Cubic>]}
          GreenCurve {keys [<at 0.0283 4.7557>,<at 0.4401 0 RCIM_Cubic>]}
          RedCurve {keys [<at 0.0283 3.1251>,<at 0.4401 0 RCIM_Cubic>]}
         }})
       1=&ColorFromCurve(ColorCurve=DataInterface() {propertyValues {
          AlphaCurve {keys [<at 0 1 RCIM_Cubic>,<at 1 1 RCIM_Cubic>]}
          BlueCurve {keys [<at 0.0394 0.9982>,<at 0.2509 0.2093>,<at 0.6029 0.0837>,<at 0.9852 0.0219>]}
          GreenCurve {keys [<at 0.0394 1.2586>,<at 0.2509 0>,<at 0.6029 0>,<at 0.9852 0>]}
          RedCurve {keys [<at 0.0394 5>,<at 0.2509 5>,<at 0.6029 0.5417>,<at 0.9852 0.151>]}
         }})
       2=&ColorFromCurve(ColorCurve=DataInterface() {propertyValues {
          AlphaCurve {keys [<at 0 1 RCIM_Cubic>,<at 1 1 RCIM_Cubic>]}
          BlueCurve {keys [<at 0.0283 5>,<at 0.4401 0 RCIM_Cubic>]}
          GreenCurve {keys [<at 0.0283 4.7557>,<at 0.4401 0 RCIM_Cubic>]}
          RedCurve {keys [<at 0.0283 3.1251>,<at 0.4401 0 RCIM_Cubic>]}
         }})
       3=?
       Index=&ModuloInt(A=$User.ColorIndex, B=2)
       Num=::NewEnumerator0
      )
      1=&MergeHelper_ColorSwitch(
       0=&ColorFromCurve(ColorCurve=DataInterface() {propertyValues {
          AlphaCurve {keys [<at 0 1 RCIM_Cubic>,<at 1 1 RCIM_Cubic>]}
          BlueCurve {keys [<at 0.0283 5>,<at 0.9819 0 RCIM_Cubic>]}
          GreenCurve {keys [<at 0.0283 4.7557>,<at 0.9819 0 RCIM_Cubic>]}
          RedCurve {keys [<at 0.0283 3.1251>,<at 0.9819 0 RCIM_Cubic>]}
         }})
       1=&ColorFromCurve(ColorCurve=DataInterface() {propertyValues {
          AlphaCurve {keys [<at 0 1 RCIM_Cubic>,<at 1 1 RCIM_Cubic>]}
          BlueCurve {keys [<at 0.0394 0.9982>,<at 0.2509 5>,<at 0.6029 0.5417>,<at 0.9852 0.151>]}
          GreenCurve {keys [<at 0.0394 1.2586>,<at 0.2509 4.6848>,<at 0.6029 0.2545>,<at 0.9852 0.025>]}
          RedCurve {keys [<at 0.0394 5>,<at 0.2509 0>,<at 0.6029 0.0591>,<at 0.9852 0.0552>]}
         }})
       2=&ColorFromCurve(ColorCurve=DataInterface() {propertyValues {
          AlphaCurve {keys [<at 0 1 RCIM_Cubic>,<at 1 1 RCIM_Cubic>]}
          BlueCurve {keys [<at 0.0283 5>,<at 0.4401 0 RCIM_Cubic>]}
          GreenCurve {keys [<at 0.0283 4.7557>,<at 0.4401 0 RCIM_Cubic>]}
          RedCurve {keys [<at 0.0283 3.1251>,<at 0.4401 0 RCIM_Cubic>]}
         }})
       3=?
       Index=&ModuloInt(A=$User.ColorIndex, B=2)
       Num=::NewEnumerator0
      )
      2=?
      3=?
      Index=&MakeIntFromFloat(Float=$User.VariantIndex)
      Num=::NewEnumerator0
     ))
    ScaleA=true
    ScaleRGB=true
   ) {module=ScaleColor001}
  ]
  rendererClasses [&NiagaraSpriteRendererProperties]
  renderers [{rendererClass=&NiagaraSpriteRendererProperties}]
 }}
summary {
 systemName=NS_Glider_CasinoReaper_Flare
 emitters [{emitterName=Flare, rendererClasses [&NiagaraSpriteRendererProperties]}]
 userVariables [
  User.ColorIndex {type=NiagaraInt32, defaultValue=1}
  User.Variant {type=NiagaraInt32, defaultValue=1}
  User.VariantIndex {type=NiagaraFloat, defaultValue=0, description=""}
 ]
}
