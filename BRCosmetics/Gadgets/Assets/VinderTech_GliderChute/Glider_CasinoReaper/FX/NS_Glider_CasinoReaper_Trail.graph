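# Niagara trail effect for the Casino Reaper glider, with separate gameplay and front-end tuning.
# Main_Trail draws anim-trail ribbons from a socket reader, and Tip_Trails spawns per-socket ribbons that blend from ForwardDot-scaled push into reversed owner velocity.
# Speed_Lines emits short-lived, speed-stretched sprites from a cylinder shell, randomly colored with one option chosen by VariantIndex.
# System time-of-day tint, fade-in and timeline values modulate color and opacity.
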NiagaraSystem /BRCosmetics/Gadgets/Assets/VinderTech_GliderChute/Glider_CasinoReaper/FX/NS_Glider_CasinoReaper_Trail
dependencies {
 usedDataInterfaces [
  &NiagaraDataInterfaceCurve
  &NiagaraDataInterfaceColorCurve
  &NiagaraDataInterfaceVector2DCurve
  &NiagaraDataInterfaceSocketReader
 ]
 usedDynamicInputs [&ColorFromCurve]
 usedModules [
  &self:SystemSpawnScript.SystemScriptSource.SystemScriptGraph.NiagaraNodeAssignment_1.SetVariables_#1
  &self:SystemSpawnScript.SystemScriptSource.SystemScriptGraph.NiagaraNodeAssignment_0.SetVariables_#2
  &SystemState
  &Timeline
  &FN_Integrator
  &self:SystemSpawnScript.SystemScriptSource.SystemScriptGraph.NiagaraNodeAssignment_4.SetVariables_#3
 ]
}
emitters {
 Main_Trail {
  emitterSpawnScript [&FN_RibbonIDOffsetUtility()]
  emitterUpdateScript [
   &EmitterState(
    Inactive Response=::NewEnumerator1
    Life Cycle Mode=::NewEnumerator1
    Loop Behavior=::NewEnumerator1
    Loop Delay=0.05
    Loop Duration Mode=::NewEnumerator1
    UseLoopDelay=true
   )
   &FN_SpawnAnimTrail()
  ]
  particleSpawnScript [
   &FN_InitializeAnimTrailBoneIndecies()
   /Game/Effects/Niagara/Modules/AnimTrail/FN_InitializeAnimTrail/FN_InitializeAnimTrail_SocketReader/FN_InitializeAnimTrail_SocketReader(
    Color:LinearColor=&MakeLinearColorFromVectorAndFloat(
     "Vector (RGB)"=&MakeVectorFromLinearColorRGB(Color=&MultiplyLinearColor(
       Linear Color=<rgba 0.4993 0.5417 0.5337 0.5>
       Scale Factor=&MakeVector4_FromLinearColor(LinearColor=$System.TOD)
      ))
     "Float (Alpha)"=&FN_FrontEndFloatModifier(Gameplay=0.5, Front End=0.2)
    )
    Lifetime=0.15
    Position Offset:Vector3f=&FN_FrontEndVector3Modifier(Gameplay=<0 0 0>, Front End=<0 0 -15>)
    Socket Reader:NiagaraDataInterfaceSocketReader=$System.Socket Reader
    Write Material Random=false
   )
   &AddVelocity(Velocity=&FN_FrontEndVector3Modifier(Gameplay=<0 2000 0>, Front End=<0 1500 200>))
   &CameraOffset(Camera Offset Amount=-500)
  ]
  particleUpdateScript [
   &InheritSourceMovement(
    Applied Position Scale=&Multiply_VectorByFloat(
     Vector=<1 1 1>
     Float=&FN_FrontEndFloatModifier(
      Gameplay=&FloatFromCurve(FloatCurve=DataInterface() {propertyValues {Curve {keys [
           {interpMode=RCIM_Cubic, tangentMode=RCTM_User, time=0.5, value=1}
           {arriveTangent=-1.9042, interpMode=RCIM_Cubic, leaveTangent=-1.9042, tangentMode=RCTM_User, time=1, value=0}
          ]}}})
      Front End=1
     )
    )
    ApplyForce=false
    ApplyPosition=true
    ApplyPositionFalloff=false
   )
   &ScaleColor(
    Color Value To Scale=$Particles.Color
    Scale Alpha=&FN_FrontEndFloatModifier(
     Gameplay=&Multiply_Float(
      A=&Lerp_Float(A=0.25, B=1, Alpha=$System.Timeline.Time)
      B=&FloatFromCurve(
       FloatCurve=DataInterface() {propertyValues {Curve {keys [{arriveTangent=-3.3639, interpMode=RCIM_Cubic, time=0, value=1}, <at 0.75 0 RCIM_Cubic>]}}}
       Scale Curve=&Multiply_Float(A=$System.SystemFadeIn, B=0.4)
      )
     )
     Front End=&Lerp_Float(A=0, B=0.4, Alpha=$System.SystemFadeIn)
    )
    ScaleA=true
    ScaleRGB=true
   ) {module=ScaleColor001}
   &SolveForcesAndVelocity(Clamp Velocity=false, Limit Acceleration=false)
   /Game/Effects/Niagara/Modules/AnimTrail/FN_UpdateAnimTrail/FN_UpdateAnimTrail_SocketReader/FN_UpdateAnimTrail_SocketReader(
    Locked Particle RibbonWidth:NiagaraFloat=&Add_Float(A=1, B=0)
    Socket Reader:NiagaraDataInterfaceSocketReader=$System.Socket Reader
   )
   /SetVariables(Particles.RibbonU0Override=$Particles.NormalizedAge) {isSetParameters=true, module=SetVariables_#4}
  ]
  rendererClasses [&NiagaraRibbonRendererProperties]
  renderers [{rendererClass=&NiagaraRibbonRendererProperties}]
 }
 Speed_Lines {
  emitterUpdateScript [
   &EmitterState()
   &SpawnRate(SpawnRate=&FN_FrontEndFloatModifier(Gameplay=100, Front End=50), Use Spawn Probability=false) {module=SpawnRate001}
  ]
  particleSpawnScript [
   &InitializeParticle(
    Color=&MultiplyLinearColor(
     Linear Color=&MultiplyLinearColor(
      Linear Color=&MergeHelper_ColorSwitch(
       0=<rgba 0 0 0 3>
       1=&MergeHelper_ColorSwitch(
        0=<rgba 1 0 0.0615 1>
        1=<rgba 0 0.7657 1 1>
        2=?
        3=?
        Index=&MakeIntFromFloat(Float=$User.VariantIndex)
        Num=::NewEnumerator0
       )
       2=<rgba 0.733 0.9995 1 1>
       3=?
       Index=&RandomRangeInt(Minimum=0, Maximum=2)
       Num=::NewEnumerator1
      )
      Scale Factor=&MakeVector4_FromLinearColor(LinearColor=$System.TOD)
     )
     Scale Factor=&MakeVector4(X=1, Y=1, Z=1, W=&FN_FrontEndFloatModifier(Gameplay=0.12, Front End=0.1))
    )
    Lifetime Max=0.16
    Lifetime Min=0.1
    Lifetime Mode=::NewEnumerator1
    Sprite Size Max=<4 10>
    Sprite Size Min=<3 5>
    Sprite Size Mode=::NewEnumerator2
    UsePositionOffset=true
    Write Material Random=true
   )
   &ShapeLocation(
    Cylinder Height=150
    Cylinder Radius=&FN_FrontEndFloatModifier(Gameplay=200, Front End=150)
    MaskSpawnGroup=false
    Offset=&FN_FrontEndVector3Modifier(Gameplay=<0 -170 30>, Front End=<0 -170 20>)
    Offset Mode=::NewEnumerator0
    Override Seed=false
    Rotation Axis=<0 1 0>
    Rotation Mode=::NewEnumerator1
    Shape Primitive=::NewEnumerator1
    Surface Only=true
    Surface Only Band Thickness=0.3
    Use Endcaps In Surface Only Mode=false
    UseOwnerScale=false
   )
   /SetVariables(Particles.Position=&Multiply_Vector(A=Linked() {linkedVariable {name=Particles.Position, type=NiagaraPosition}}, B=<1.1 1 0.9>)) {isSetParameters=true}
   &AddVelocityFromPoint(
    Module.Use Falloff Distance=false
    Origin Offset=<0 -1000 -250>
    Velocity Strength=&FN_FrontEndFloatModifier(Gameplay=1000, Front End=0)
   )
   &AddVelocity(
    Velocity=&FN_FrontEndVector3Modifier(Gameplay=<0 300 0>, Front End=<0 300 100>)
    Velocity Speed Scale=&RandomRangeFloat(Minimum=0.5, Maximum=2)
   ) {module=AddVelocity001}
  ]
  particleUpdateScript [
   &ParticleState()
   &ScaleSpriteSize(Scale Factor=&Vector2DFromCurve(
     Vector2Curve=DataInterface() {propertyValues {XCurve {keys [<at 0 0.5 RCIM_Cubic>,<at 0.5 1 RCIM_Cubic>]}, YCurve {keys [<at 0 0 RCIM_Cubic>,<at 0.5 1 RCIM_Cubic>]}}}
     ScaleCurve=&Vector2DFromFloat(Value=1)
    ))
   &SolveForcesAndVelocity(Clamp Velocity=false, Limit Acceleration=false)
   &ScaleSpriteSizeBySpeed(Max Scale Factor=<1 50>, Min Scale Factor=<0.01 0.05>, Velocity Threshold=1)
   &ScaleColor(
    Scale Alpha=&FloatFromCurve(
     FloatCurve=DataInterface() {propertyValues {Curve {keys [
         {arriveTangent=6.8397, interpMode=RCIM_Cubic, leaveTangent=6.8397, tangentMode=RCTM_User, time=0, value=0}
         <at 0.3 1 RCIM_Cubic>
         {arriveTangent=-2.221, interpMode=RCIM_Cubic, leaveTangent=-2.221, tangentMode=RCTM_User, time=1, value=0}
        ]}}}
     Scale Curve=&FN_FrontEndFloatModifier(
      Gameplay=&Lerp_Float(
       A=0
       B=&FNTimeOfDayScalar(Night Scale=0.2, Morning Scale=0.3, Day Scale=0.4, Evening Scale=0.3)
       Alpha=&Multiply_Float(A=$System.SystemFadeIn, B=$System.Timeline.Time)
      )
      Front End=&Lerp_Float(A=0, B=0.4, Alpha=$System.SystemFadeIn)
     )
    )
    Scale RGB=&VectorFromFloat(Value=&FN_FrontEndFloatModifier(Gameplay=1, Front End=1))
    ScaleA=true
    ScaleRGB=true
   )
  ]
  rendererClasses [&NiagaraSpriteRendererProperties]
  renderers [{rendererClass=&NiagaraSpriteRendererProperties}]
 }
 Tip_Trails {
  emitterSpawnScript [&FN_RibbonIDOffsetUtility()]
  emitterUpdateScript [
   &EmitterState(Life Cycle Mode=::NewEnumerator1, Loop Behavior=::NewEnumerator1, Loop Duration Mode=::NewEnumerator1, UseLoopDelay=false)
   &SpawnRate(SpawnRate=240, Use Spawn Probability=false)
  ]
  particleSpawnScript [
   &PartitionParticles(Write Partition to Ribbon ID=::NewEnumerator0)
   &InitializeParticle(
    Color=&MakeLinearColorFromVectorAndFloat(
     "Vector (RGB)"=&MakeVectorFromLinearColorRGB(Color=&MultiplyLinearColor(
       Linear Color=<rgba 0.5885 0.5885 0.5885 0.3>
       Scale Factor=&MakeVector4_FromLinearColor(LinearColor=$System.TOD)
      ))
     "Float (Alpha)"=&FN_FrontEndFloatModifier(Gameplay=0.3, Front End=0.2)
    )
    Lifetime=0.3
    Material Random:NiagaraFloat=&Add_Float(A=0, B=0)
    Ribbon Width=15
    Ribbon Width Mode=::NewEnumerator1
    UsePositionOffset=false
    Write Material Random=true
   )
   &SocketLocation(
    SamplePosOffset=true
    Sampled Position Offset=&FN_FrontEndVector3Modifier(Gameplay=<0 0 0>, Front End=<0 0 -15>)
    Socket Index=$Particles.PartitionParticles.Partition
    Socket Reader=$System.Socket Reader
    Socket Sampling Mode=::NewEnumerator3
    VelocitySpeedLimit=false
   ) {module=SocketLocation001}
   /SetVariables(Particles.InitForward:NiagaraFloat=$User.ForwardDot) {isSetParameters=true}
   &CameraOffset(Camera Offset Amount=-50)
  ]
  particleUpdateScript [
   &ParticleState()
   /SetVariables(Particles.Velocity=&FN_FrontEndVector3Modifier(
     Gameplay=&Lerp_Vector(
      A=&Multiply_VectorByFloat(Vector=<0 3000 0>, Float=$Particles.InitForward)
      B=&Multiply_VectorByFloat(
       Vector=&TransformVector(Vector=$Engine.Owner.Velocity, Source Space=::World, Destination Space=::Local)
       Float=-1
      )
      Alpha=&FloatFromCurve(FloatCurve=DataInterface() {propertyValues {Curve {keys [{arriveTangent=1.9495, time=0, value=0}, {leaveTangent=2.1267, time=0.5, value=1}]}}})
     )
     Front End=<0 3000 500>
    )) {isSetParameters=true}
   /SetVariables(Particles.RibbonU0Override=&Multiply_Float(A=$Particles.NormalizedAge, B=1)) {isSetParameters=true, module=SetVariables_#4}
   &ScaleColor(
    Scale Alpha=&Multiply_Float(
     A=&FN_FrontEndFloatModifier(
      Gameplay=&Multiply_Float(
       A=&Multiply_Float(A=$System.SystemFadeIn, B=0.2)
       B=&Lerp_Float(A=0.25, B=1, Alpha=$System.Timeline.Time)
      )
      Front End=&Lerp_Float(A=0, B=0.2, Alpha=$System.SystemFadeIn)
     )
     B=&Multiply_Float(
      A=&FNTimeOfDayScalar(Night Scale=0.6, Morning Scale=0.8, Evening Scale=0.8)
      B=&FloatFromCurve(FloatCurve=DataInterface() {propertyValues {Curve {keys [<at 0 1>,<at 0.4 0>]}}})
     )
    )
    ScaleA=true
    ScaleRGB=true
   ) {module=ScaleColor001}
   &DynamicMaterialParameters(
    Index 0 Param 1=&FloatFromCurve(FloatCurve=DataInterface() {propertyValues {Curve {keys [
         {interpMode=RCIM_Cubic, tangentMode=RCTM_User, time=0.2, value=1}
         {arriveTangent=-2.3194, interpMode=RCIM_Cubic, leaveTangent=-2.3194, tangentMode=RCTM_User, time=0.5, value=0}
        ]}}})
    Index 0 Param 3=-3
    Module.Index0Float0Write=true
    Module.Index0Float1Write=true
    Module.Index0Float2Write=true
    Module.Index0Float3Write=true
   )
   &SolveForcesAndVelocity(Clamp Velocity=false, Limit Acceleration=false)
   &KillParticles(Kill Particles=&SetBoolByFloatComparison(A=$Particles.NormalizedAge, B=1)) {enabled=false}
  ]
  rendererClasses [&NiagaraRibbonRendererProperties]
  renderers [{rendererClass=&NiagaraRibbonRendererProperties}]
 }
}
summary {
 systemName=NS_Glider_CasinoReaper_Trail
 emitters [
  {emitterName=Main_Trail, rendererClasses [&NiagaraRibbonRendererProperties]}
  {emitterName=Speed_Lines, rendererClasses [&NiagaraSpriteRendererProperties]}
  {emitterName=Tip_Trails, rendererClasses [&NiagaraRibbonRendererProperties]}
 ]
 userVariables [
  User.bIsFrontEnd {type=NiagaraBool}
  User.bIsFrontEndPreview {type=NiagaraBool}
  User.ForwardDot {type=NiagaraFloat, defaultValue=0, description=""}
  User.RibbonIDOffset {type=NiagaraInt32}
  User.RightDot {type=NiagaraFloat, defaultValue=0, description=""}
  User.VariantIndex {type=NiagaraFloat, defaultValue=0, description=""}
 ]
}
